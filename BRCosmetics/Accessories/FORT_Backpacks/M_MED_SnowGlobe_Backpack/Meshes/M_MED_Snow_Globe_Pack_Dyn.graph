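# Physics asset for the M_MED Snow Globe back bling, setting up simulation bodies for a dynamic tree bone chain on the pack.
# The pack bone carries a small capsule and a rotated, offset box, while four tree bones each use a small capsule.
# Constraints chain the bones in sequence: the tree root swings and twists freely against the pack, and higher segments carry tight angular limits.
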PhysicsAsset /BRCosmetics/Accessories/FORT_Backpacks/M_MED_SnowGlobe_Backpack/Meshes/M_MED_Snow_Globe_Pack_Dyn
bodies {
 dyn_tree_01 [{shapeType=Capsule, length=1, radius=1, rotation=<rot 90 0 0>}]
 dyn_tree_02 [{shapeType=Capsule, length=2, radius=1, center=<0.5 0 0>, rotation=<rot 90 0 0>}]
 dyn_tree_03 [{shapeType=Capsule, length=2, radius=1, center=<0.5 0 0>, rotation=<rot 90 0 0>}]
 dyn_tree_04 [{shapeType=Capsule, length=2, radius=1, center=<0.5 0 0>, rotation=<rot 90 0 0>}]
 pack [
  {shapeType=Capsule, length=1, radius=1}
  {shapeType=Box, extentX=7.5, extentY=7.5, extentZ=7.5, center=<-4.0114 -8.6212 -4.8258>, rotation=<rot 5 39.9959 -5.0013>}
 ]
}
constraints [
 {
  bone1Name=dyn_tree_01
  bone2Name=pack
  swing1LimitDegrees=45
  swing1Motion=Free
  swing2LimitDegrees=45
  swing2Motion=Free
  twistLimitDegrees=45
  twistMotion=Free
 }
 {bone1Name=dyn_tree_02, bone2Name=dyn_tree_01, swing1LimitDegrees=15, swing2LimitDegrees=5, twistLimitDegrees=2}
 {bone1Name=dyn_tree_03, bone2Name=dyn_tree_02, swing1LimitDegrees=15, swing2LimitDegrees=5, twistLimitDegrees=2}
 {bone1Name=dyn_tree_04, bone2Name=dyn_tree_03, swing1LimitDegrees=15, swing2LimitDegrees=5, twistLimitDegrees=2}
]
